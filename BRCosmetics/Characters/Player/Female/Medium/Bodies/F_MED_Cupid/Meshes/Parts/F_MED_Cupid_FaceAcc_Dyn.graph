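# Physics asset for the Cupid medium female character's FaceAcc part, simulating its dynamic hair.
# Seven hair bone chains of mostly boxes, with some capsule tips, hang from the head with per-joint swing limits and tight twist limits.
# Head, neck, upper spine, clavicle and upper-arm capsules form the body volumes, joined by angularly free constraints.
# Extra angularly free constraints cross-link the right, back-right and back chains near their ends.
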PhysicsAsset /BRCosmetics/Characters/Player/Female/Medium/Bodies/F_MED_Cupid/Meshes/Parts/F_MED_Cupid_FaceAcc_Dyn
bodies {
 clavicle_l [{shapeType=Capsule, length=4.6352, radius=6.8578, center=<6.6391 -0.0942 -2.3232>, rotation=<rot 90 -162.646 -167.646>}]
 clavicle_r [{shapeType=Capsule, length=9.5687, radius=6.3578, center=<-5.1339 1.3128 1.3915>, rotation=<rot 71.985 5.0001 0>}]
 dyn_hair_bk_01 [{shapeType=Box, extentX=6.4375, extentY=8.125, extentZ=1.0625, center=<3.2204 0 0>}]
 dyn_hair_bk_02 [{shapeType=Box, extentX=6.4375, extentY=8.125, extentZ=3.9337, center=<3.2204 0 0.2941>}]
 dyn_hair_bk_03 [{shapeType=Box, extentX=6.4375, extentY=15.125, extentZ=5.149, center=<3.2204 -0.002 1.4073>}]
 dyn_hair_bk_04 [{shapeType=Box, extentX=6.4375, extentY=14.125, extentZ=4.1861, center=<3.2812 0 -0.1239>, rotation=<rot -13.1142 0 0>}]
 dyn_hair_bk_l_01 [{shapeType=Box, extentX=6.4375, extentY=8.125, extentZ=1.25, center=<3.2204 0 0>}]
 dyn_hair_bk_l_02 [{shapeType=Box, extentX=4.125, extentY=8.125, extentZ=3.0625, center=<1.8061 0 0>}]
 dyn_hair_bk_l_03 [{shapeType=Box, extentX=3.943, extentY=7.943, extentZ=2.1875, center=<-0.7275 0 0.6065>, rotation=<rot -0.0002 0.0001 0>}]
 dyn_hair_bk_r_01 [{shapeType=Box, extentX=4.5933, extentY=8.125, extentZ=1.25, center=<1.901 0.0001 0.0004>}]
 dyn_hair_bk_r_02 [{shapeType=Box, extentX=15.3824, extentY=13.3606, extentZ=1.5924, center=<5.907 0.0538 2.2554>, rotation=<rot 14.8074 0 5.0001>}]
 dyn_hair_bk_r_03 [{shapeType=Box, extentX=8.6919, extentY=18.0667, extentZ=2.1875, center=<0 1.6851 2.8113>, rotation=<rot -0.0002 -0.0001 30.9375>}]
 dyn_hair_fr_l_01 [{shapeType=Box, extentX=5.4375, extentY=7.125, extentZ=0.5, center=<3.2204 0 0>}]
 dyn_hair_fr_l_02 [{shapeType=Box, extentX=5.4375, extentY=7.125, extentZ=0.5, center=<3.2204 0 0>}]
 dyn_hair_fr_l_03 [
  {shapeType=Capsule, length=2.2329, radius=1.8098, center=<5.5443 3.7864 0.3346>, rotation=<rot -10.262 123.9831 -96.8482>}
  {shapeType=Capsule, length=1.9014, radius=1.361, center=<1.8673 1.6276 0.3525>, rotation=<rot 90 0 119.9984>}
 ]
 dyn_hair_fr_r_01 [{shapeType=Box, extentX=6.4375, extentY=8.125, extentZ=1.0625, center=<3.2204 0 0>}]
 dyn_hair_fr_r_02 [{shapeType=Box, extentX=6.4375, extentY=8.125, extentZ=1.375, center=<3.2204 0 0>}]
 dyn_hair_fr_r_03 [
  {shapeType=Capsule, length=2.5, radius=1.7812, center=<2.0129 -2.32 1.2519>, rotation=<rot 8.8304 45.3664 -81.3776>}
  {shapeType=Capsule, length=3.2847, radius=1.5354, center=<1.3963 -1.1406 0>, rotation=<rot -0.0004 24.9995 -89.9992>}
 ]
 dyn_hair_l_01 [{shapeType=Box, extentX=6.4375, extentY=8.125, extentZ=1.875, center=<3.2204 0 0>}]
 dyn_hair_l_02 [{shapeType=Box, extentX=6.4375, extentY=8.125, extentZ=2.3125, center=<3.2204 0 0>}]
 dyn_hair_l_03 [{shapeType=Capsule, length=2.8815, radius=1.2122, rotation=<rot 0 0.0001 -87.1875>}]
 dyn_hair_r_01 [{shapeType=Box, extentX=3.1158, extentY=8.125, extentZ=1.875, center=<1.593 0 0>}]
 dyn_hair_r_02 [{shapeType=Box, extentX=7.789, extentY=10.8964, extentZ=2.3125, center=<4.6988 0.0002 0>}]
 dyn_hair_r_03 [{shapeType=Box, extentX=14.4571, extentY=17.0788, extentZ=2.6438, center=<3.502 0.4585 5.7792>, rotation=<rot 47.5699 1.8631 18.81>}]
 head [{shapeType=Capsule, length=5.5, radius=6.4062, center=<0.0001 -0.4045 0>, rotation=<rot 90 0 0>}]
 neck_01 [{shapeType=Capsule, length=4.5312, radius=4.125, center=<0.9973 -0.7016 0>, rotation=<rot 90 0 0>}]
 spine_05 [
  {shapeType=Capsule, length=17.3721, radius=7.1155, center=<3.4024 -0.3475 0>}
  {shapeType=Capsule, length=12.2023, radius=5.616, shapeName=spine_05_capsule, center=<-0.8196 4.7933 0>}
 ]
 upperarm_l [{shapeType=Capsule, length=9, radius=4.25, center=<8.9828 0.0001 -0.8582>, rotation=<rot -90 0.0009 -0.0001>}]
 upperarm_r [{shapeType=Capsule, length=9, radius=4.25, center=<-8.9828 -0.0002 0.8582>, rotation=<rot 90 -160.529 -160.5303>}]
}
constraints [
 {bone1Name=dyn_hair_bk_02, bone2Name=dyn_hair_bk_01, swing1LimitDegrees=15, swing2LimitDegrees=60, twistLimitDegrees=5}
 {bone1Name=dyn_hair_bk_03, bone2Name=dyn_hair_bk_02, swing1LimitDegrees=15, swing2LimitDegrees=60, twistLimitDegrees=5}
 {bone1Name=dyn_hair_bk_04, bone2Name=dyn_hair_bk_03, swing1LimitDegrees=15, swing2LimitDegrees=60, twistLimitDegrees=5}
 {bone1Name=dyn_hair_r_02, bone2Name=dyn_hair_r_01, swing1LimitDegrees=60, swing2LimitDegrees=60, twistLimitDegrees=5}
 {bone1Name=dyn_hair_r_03, bone2Name=dyn_hair_r_02, swing1LimitDegrees=45, swing2LimitDegrees=45, twistLimitDegrees=5}
 {bone1Name=dyn_hair_bk_r_02, bone2Name=dyn_hair_bk_r_01, swing1LimitDegrees=20, swing2LimitDegrees=60, twistLimitDegrees=5}
 {bone1Name=dyn_hair_bk_r_03, bone2Name=dyn_hair_bk_r_02, swing1LimitDegrees=20, swing2LimitDegrees=30, twistLimitDegrees=5}
 {bone1Name=dyn_hair_fr_r_02, bone2Name=dyn_hair_fr_r_01, swing1LimitDegrees=80, swing2LimitDegrees=80, twistLimitDegrees=10}
 {bone1Name=dyn_hair_fr_r_03, bone2Name=dyn_hair_fr_r_02, swing1LimitDegrees=80, swing2LimitDegrees=80, twistLimitDegrees=5}
 {bone1Name=dyn_hair_bk_l_02, bone2Name=dyn_hair_bk_l_01, swing1LimitDegrees=10, swing2LimitDegrees=30, twistLimitDegrees=5}
 {bone1Name=dyn_hair_bk_l_03, bone2Name=dyn_hair_bk_l_02, swing1LimitDegrees=10, swing2LimitDegrees=30, twistLimitDegrees=5}
 {bone1Name=dyn_hair_l_02, bone2Name=dyn_hair_l_01, swing1LimitDegrees=45, swing2LimitDegrees=45, twistLimitDegrees=5}
 {bone1Name=dyn_hair_l_03, bone2Name=dyn_hair_l_02, swing1LimitDegrees=45, swing2LimitDegrees=45, twistLimitDegrees=5}
 {bone1Name=dyn_hair_fr_l_02, bone2Name=dyn_hair_fr_l_01, swing1LimitDegrees=80, swing2LimitDegrees=80, twistLimitDegrees=10}
 {bone1Name=dyn_hair_fr_l_03, bone2Name=dyn_hair_fr_l_02, swing1LimitDegrees=80, swing2LimitDegrees=80, twistLimitDegrees=5}
 {bone1Name=dyn_hair_bk_01, bone2Name=head, swing1LimitDegrees=15, swing2LimitDegrees=60, twistLimitDegrees=5}
 {bone1Name=dyn_hair_r_01, bone2Name=head, swing1LimitDegrees=45, swing2LimitDegrees=45, twistLimitDegrees=5}
 {bone1Name=dyn_hair_bk_r_01, bone2Name=head, swing1LimitDegrees=20, swing2LimitDegrees=60, twistLimitDegrees=5}
 {bone1Name=dyn_hair_fr_r_01, bone2Name=head, swing1LimitDegrees=45, swing2LimitDegrees=45, twistLimitDegrees=5}
 {bone1Name=dyn_hair_bk_l_01, bone2Name=head, swing1LimitDegrees=10, swing2LimitDegrees=60, twistLimitDegrees=5}
 {bone1Name=dyn_hair_l_01, bone2Name=head, swing1LimitDegrees=45, swing2LimitDegrees=45, twistLimitDegrees=5}
 {bone1Name=dyn_hair_fr_l_01, bone2Name=head, swing1LimitDegrees=45, swing2LimitDegrees=45, twistLimitDegrees=5}
 {
  bone1Name=head
  bone2Name=neck_01
  swing1LimitDegrees=45
  swing1Motion=Free
  swing2LimitDegrees=45
  swing2Motion=Free
  twistLimitDegrees=45
  twistMotion=Free
 }
 {
  bone1Name=clavicle_l
  bone2Name=spine_05
  swing1LimitDegrees=45
  swing1Motion=Free
  swing2LimitDegrees=45
  swing2Motion=Free
  twistLimitDegrees=45
  twistMotion=Free
 }
 {
  bone1Name=clavicle_r
  bone2Name=spine_05
  swing1LimitDegrees=45
  swing1Motion=Free
  swing2LimitDegrees=45
  swing2Motion=Free
  twistLimitDegrees=45
  twistMotion=Free
 }
 {
  bone1Name=neck_01
  bone2Name=spine_05
  swing1LimitDegrees=45
  swing1Motion=Free
  swing2LimitDegrees=45
  swing2Motion=Free
  twistLimitDegrees=45
  twistMotion=Free
 }
 {
  bone1Name=dyn_hair_r_03
  bone2Name=dyn_hair_bk_r_03
  swing1LimitDegrees=45
  swing1Motion=Free
  swing2LimitDegrees=45
  swing2Motion=Free
  twistLimitDegrees=45
  twistMotion=Free
 }
 {
  bone1Name=dyn_hair_bk_r_03
  bone2Name=dyn_hair_bk_03
  swing1LimitDegrees=45
  swing1Motion=Free
  swing2LimitDegrees=45
  swing2Motion=Free
  twistLimitDegrees=45
  twistMotion=Free
 }
 {
  bone1Name=upperarm_l
  bone2Name=clavicle_l
  swing1LimitDegrees=45
  swing1Motion=Free
  swing2LimitDegrees=45
  swing2Motion=Free
  twistLimitDegrees=45
  twistMotion=Free
 }
 {
  bone1Name=upperarm_r
  bone2Name=clavicle_r
  swing1LimitDegrees=45
  swing1Motion=Free
  swing2LimitDegrees=45
  swing2Motion=Free
  twistLimitDegrees=45
  twistMotion=Free
 }
]
