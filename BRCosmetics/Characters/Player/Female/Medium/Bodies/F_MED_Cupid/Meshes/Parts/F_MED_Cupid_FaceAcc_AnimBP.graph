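# Animation Blueprint for the Cupid female face-accessory mesh, adding secondary motion to its dyn_hair bones.
# AnimGraph copies the parent mesh pose, simulates RigidBody physics relative to neck_01, caches it, then applies chained additive bone-space offsets.
# EventGraph computes gravity for swim-sprinting, skydiving and parachuting, plus frontend and head-up alphas.
# The shown AnimGraph reads none of these variables, and its front-end blend node is disconnected from the output.
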AnimBlueprint /BRCosmetics/Characters/Player/Female/Medium/Bodies/F_MED_Cupid/Meshes/Parts/F_MED_Cupid_FaceAcc_AnimBP
target_skeleton=&/F_MED_Cupid_FaceAcc_Skeleton
graphs [
 AnimGraph {nodes [
   Blend Poses by bool(BlendPose_1=Use cached pose 'Pose'(), bActiveValue=Get bIsInFrontEnd()) {blendType=Linear}
   n1 = "Transform (Modify) Bone - Bone: dyn_hair_fr_l_02"!(ComponentPose=Local To Component(LocalPose=Use cached pose 'Pose'()), Alpha=1, Translation="0,0,-3", Rotation="20,0,0") -> (Pose) {
    rotationMode=BMM_Additive
    rotationSpace=BCS_BoneSpace
    translationMode=BMM_Additive
    translationSpace=BCS_BoneSpace
    alphaBoolBlend {blendOption=Linear}
    alphaScaleBiasClamp {bInterpResult=true, interpSpeedDecreasing=5, interpSpeedIncreasing=7}
    boneToModify {boneName=dyn_hair_fr_l_02}
    rotation {pitch=20}
    translation {z=-3}
   }
   n2 = "Transform (Modify) Bone - Bone: dyn_hair_fr_r_02"!(ComponentPose=@n1, Alpha=1, Translation="0,0,0", Rotation="-5,-0.000,0.000") -> (Pose) {
    rotationMode=BMM_Additive
    rotationSpace=BCS_BoneSpace
    translationMode=BMM_Additive
    translationSpace=BCS_BoneSpace
    alphaBoolBlend {blendOption=Linear}
    alphaScaleBiasClamp {bInterpResult=true, interpSpeedDecreasing=5, interpSpeedIncreasing=7}
    boneToModify {boneName=dyn_hair_fr_r_02}
    rotation {pitch=-5}
   }
   n3 = "Transform (Modify) Bone - Bone: dyn_hair_bk_01"!(ComponentPose=@n2, Alpha=1, Translation="-0.092,-0.005,1", Rotation="4.981,5.019,0.437") -> (Pose) {
    rotationMode=BMM_Additive
    rotationSpace=BCS_BoneSpace
    translationMode=BMM_Additive
    translationSpace=BCS_BoneSpace
    alphaBoolBlend {blendOption=Linear}
    alphaScaleBiasClamp {bInterpResult=true, interpSpeedDecreasing=5, interpSpeedIncreasing=7}
    boneToModify {boneName=dyn_hair_bk_01}
    rotation=<rot 4.9809 5.019 0.437>
    translation=<-0.092 -0.005 1>
   }
   n4 = "Transform (Modify) Bone - Bone: dyn_hair_bk_01"(ComponentPose=@n3, Alpha=1, Translation="0,0,0", Rotation="0,0,0") {
    rotationMode=BMM_Additive
    rotationSpace=BCS_BoneSpace
    translationMode=BMM_Additive
    translationSpace=BCS_BoneSpace
    alphaBoolBlend {blendOption=Linear}
    alphaScaleBiasClamp {bInterpResult=true, interpSpeedDecreasing=5, interpSpeedIncreasing=7}
    boneToModify {boneName=dyn_hair_bk_01}
   }
   n5 = "Transform (Modify) Bone - Bone: dyn_hair_bk_r_02"!(ComponentPose=@n4, Alpha=1, Translation="0,0,0", Rotation="0,0,0") -> (Pose) {
    rotationMode=BMM_Additive
    rotationSpace=BCS_BoneSpace
    translationMode=BMM_Additive
    translationSpace=BCS_BoneSpace
    alphaBoolBlend {blendOption=Linear}
    alphaScaleBiasClamp {bInterpResult=true, interpSpeedDecreasing=5, interpSpeedIncreasing=7}
    boneToModify {boneName=dyn_hair_bk_r_02}
   }
   n6 = "Transform (Modify) Bone - Bone: dyn_hair_bk_03"!(ComponentPose=@n5, Alpha=1, Translation="0,0,0", Rotation="0,0,0") -> (Pose) {
    rotationMode=BMM_Additive
    rotationSpace=BCS_BoneSpace
    translationMode=BMM_Additive
    translationSpace=BCS_BoneSpace
    alphaBoolBlend {blendOption=Linear}
    alphaScaleBiasClamp {bInterpResult=true, interpSpeedDecreasing=5, interpSpeedIncreasing=7}
    boneToModify {boneName=dyn_hair_bk_03}
   }
   n7 = "Transform (Modify) Bone - Bone: dyn_hair_bk_r_02"(ComponentPose=@n6, Alpha=1, Translation="-3,0,0", Rotation="20,0,0") {
    rotationMode=BMM_Additive
    rotationSpace=BCS_BoneSpace
    translationMode=BMM_Additive
    translationSpace=BCS_BoneSpace
    alphaBoolBlend {blendOption=Linear}
    alphaScaleBiasClamp {bInterpResult=true, interpSpeedDecreasing=5, interpSpeedIncreasing=7}
    boneToModify {boneName=dyn_hair_bk_r_02}
    rotation {pitch=20}
    translation {x=-3}
   }
   n8 = "Transform (Modify) Bone - Bone: dyn_hair_fr_r_02"(ComponentPose=@n7, Alpha=1, Translation="-1,0,5", Rotation="0,0,0") {
    rotationMode=BMM_Additive
    rotationSpace=BCS_BoneSpace
    translationMode=BMM_Additive
    translationSpace=BCS_BoneSpace
    alphaBoolBlend {blendOption=Linear}
    alphaScaleBiasClamp {bInterpResult=true, interpSpeedDecreasing=5, interpSpeedIncreasing=7}
    boneToModify {boneName=dyn_hair_fr_r_02}
    translation {x=-1, z=5}
   }
   n9 = "Transform (Modify) Bone - Bone: dyn_hair_r_02"!(ComponentPose=@n8, Alpha=1, Translation="-2,0,0", Rotation="0,0,0") -> (Pose) {
    rotationMode=BMM_Additive
    rotationSpace=BCS_BoneSpace
    translationMode=BMM_Additive
    translationSpace=BCS_BoneSpace
    alphaBoolBlend {blendOption=Linear}
    alphaScaleBiasClamp {bInterpResult=true, interpSpeedDecreasing=5, interpSpeedIncreasing=7}
    boneToModify {boneName=dyn_hair_r_02}
    translation {x=-2}
   }
   n10 = "Transform (Modify) Bone - Bone: dyn_hair_bk_r_02"(ComponentPose=@n9, Alpha=1, Translation="-2,0,0", Rotation="0,0,0") {
    rotationMode=BMM_Additive
    rotationSpace=BCS_BoneSpace
    translationMode=BMM_Additive
    translationSpace=BCS_BoneSpace
    alphaBoolBlend {blendOption=Linear}
    alphaScaleBiasClamp {bInterpResult=true, interpSpeedDecreasing=5, interpSpeedIncreasing=7}
    boneToModify {boneName=dyn_hair_bk_r_02}
    translation {x=-2}
   }
   Output Pose(Result=Component To Local(ComponentPose=@n10)) {layerGroup=None}
   n11 = RigidBody!(
    ComponentPose=Local To Component(LocalPose=Copy Pose From Mesh() {bUseAttachedParent=true})
    Alpha
    OverrideWorldGravity:Vector="0,0,-980"
    ExternalForce
   ) -> (Pose) {
    bClampLinearTranslationLimitToRefPose=true
    bOverrideWorldGravity=true
    overridePhysicsAsset=&/F_MED_Cupid_FaceAcc_Dyn
    simulationSpace=BaseBoneSpace
    alphaBoolBlend {blendOption=Linear}
    baseBoneRef {boneName=neck_01}
    componentAppliedLinearAccClamp=<2000 2000 2000>
    componentLinearAccScale=<0.1 0.07 0.07>
    componentLinearVelScale=<0.1 0.07 0.07>
    overrideWorldGravity {z=-980}
   }
   Save cached pose 'Pose'(Pose=Component To Local(ComponentPose=@n11)) {cacheName=Pose}
  ]}
 EventGraph {nodes [
   Event BlueprintUpdateAnimation()
   > n1 = Branch(Condition=Get bIsSwimSprinting() -> (bIsSwimSprinting:bool))
   > Set gravity_override(gravity_override="0, 0, 0")
   n2 = Branch(execute=@n1.else, Condition=Get bIsSkyDiving())
   > n3 = Branch(Condition=Get bIsParachuteOpened())
   Set gravity_override(execute=[@n2.else, @n3.then], gravity_override="0, 0,-980")
   > n4 = Branch(Condition=IsCosmeticObjectDisplayedInFrontendContext!(CosmeticObject:Object Reference) -> (ReturnValue:bool))
   > n5 = Set FE Alpha(FE Alpha:double=0.8) -> (Output_Get:double)
   n6 = GetOwnerSkeletalMeshComponent()
   n7 = BreakRotator(InRot=BreakTransform(InTransform=GetDeltaTransformFromRefPose(self=@n6, BoneName=head, BaseName=spine_05)).Rotation)
   Set Head Up(execute=[@n9.then, @n5.then], Head Up:double=MapRangeClamped(Value=@n7.Yaw, InRangeA=-20, InRangeB=-35)) -> (Output_Get:double)
   n8 = Get bIsTargeting()
   > Set Head Up Targeting(Head Up Targeting:double=float * float(A=MapRangeClamped(Value=@n7.Yaw, InRangeA=0, InRangeB=-17), B="To Float (Boolean)"(InBool=@n8)) {is=title K2Node_PromotableOperator}) -> (Output_Get:double)
   n9 = Set FE Alpha(execute=@n4.else, FE Alpha=1)
   n10 = Branch(execute=@n3.else, Condition="float > float"(A=Get PawnVelocityZ(), B=50) {is=title K2Node_CallFunction})
   > Set gravity_override(gravity_override="0, 0,200")
   n11 = Get Skel_Mesh()
   n12 = vector * float(
    A=GetForwardVector!(InRot:Rotator=GetSocketRotation!(self:Scene Component Object Reference=@n11, InSocketName:Name=pelvis) -> (ReturnValue:Rotator)) -> (ReturnValue:Vector)
    B=-1000
   )
   Set gravity_override(
    execute=@n10.else
    gravity_override=vector + vector(A:Vector=@n12, B:Vector="0, 0,-200") -> (ReturnValue:Vector) {is=title K2Node_CommutativeAssociativeBinaryOperator}
   )
   Event BlueprintInitializeAnimation()
   > Set Skel_Mesh(Skel_Mesh=GetOwnerSkeletalMeshComponent())
   TryGetPawnOwner()
   n13 = GetSocketLocation!(self:Scene Component Object Reference=@n11, InSocketName:Name=pelvis) -> (ReturnValue:Vector)
   DrawDebugLine!(
    execute
    LineStart:Vector=@n13
    LineEnd:Vector=vector + vector(A=@n13, B=@n12)
    LineColor:Linear Color Structure=(R=1,G=0,B=0,A=1)
    Duration:float=0
    Thickness:float=3
    DepthPriority:EDrawDebugSceneDepthPriorityGroup Enum=World
   ) -> (then)
  ]}
]
variables [
 FE Alpha {type=number}
 gravity_override {type=Vector}
 Head Up {type=number}
 Head Up Targeting {type=number}
 Skel_Mesh {type=SkeletalMeshComponent}
]
